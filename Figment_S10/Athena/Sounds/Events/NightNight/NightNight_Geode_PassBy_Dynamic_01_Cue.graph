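# Sound cue for a geode pass-by in the NightNight event, playing a randomly chosen space-rock whoosh.
# A quality-level switch sends the first two levels to a five-variant random pool and the last two to a two-variant pool.
# All but one whoosh in each pool start after a short fixed delay.
# It uses a high-priority events sound class and an attenuation override with distance volume falloff turned off.
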SoundCue /Figment_S10/Athena/Sounds/Events/NightNight/NightNight_Geode_PassBy_Dynamic_01_Cue
root=QualityLevel_0
nodes [
 {id=Delay_0, children [WavePlayer_12], properties {delayMax=1, delayMin=1}}
 {id=Delay_1, children [WavePlayer_4], properties {delayMax=0.5, delayMin=0.5}}
 {id=Delay_2, children [WavePlayer_11], properties {delayMax=0.5, delayMin=0.5}}
 {id=Delay_3, children [WavePlayer_10], properties {delayMax=0.5, delayMin=0.5}}
 {id=Delay_4, children [WavePlayer_1], properties {delayMax=0.5, delayMin=0.5}}
 {id=QualityLevel_0, children [Random_0,Random_0,Random_1,Random_1]}
 {id=Random_0, children [WavePlayer_2,Delay_3,Delay_1,Delay_2,Delay_0], properties {weights [1,1,1,1,1]}}
 {id=Random_1, children [WavePlayer_0,Delay_4], properties {bShouldExcludeFromBranchCulling=true, weights [1,1]}}
 {id=WavePlayer_0, properties {soundWaveAssetPtr=/Figment_S10/Sounds/Fort_GamePlay_Sounds/Wargames/SpaceRocks_Whoosh_01}}
 {id=WavePlayer_1, properties {soundWaveAssetPtr=/Figment_S10/Sounds/Fort_GamePlay_Sounds/Wargames/SpaceRocks_Whoosh_06}}
 {id=WavePlayer_10, properties {soundWaveAssetPtr=&SpaceRocks_Whoosh_06}}
 {id=WavePlayer_11, properties {soundWaveAssetPtr=/Figment_S10/Sounds/Fort_GamePlay_Sounds/Wargames/SpaceRocks_Whoosh_07}}
 {id=WavePlayer_12, properties {soundWaveAssetPtr=/Figment_S10/Sounds/Fort_GamePlay_Sounds/Wargames/SpaceRocks_Whoosh_08}}
 {id=WavePlayer_2, properties {soundWaveAssetPtr=&SpaceRocks_Whoosh_01}}
 {id=WavePlayer_4, properties {soundWaveAssetPtr=/Figment_S10/Sounds/Fort_GamePlay_Sounds/Wargames/SpaceRocks_Whoosh_03}}
]
properties {
 bOverrideAttenuation=true
 maxDistance=2097152
 soundClassObject=/Figment_S10/Athena/Sounds/Events/SoundClasses/Fort_Events_Priority_ExtHigh
 volumeMultiplier=1
 attenuationOverrides {
  bAttenuate=false
  bEnableFocusInterpolation=true
  falloffDistance=100000
  focusAttackInterpSpeed=0.5
  focusAzimuth=5
  focusReleaseInterpSpeed=0.25
  lPFFrequencyAtMax=500
  lPFRadiusMax=2000
  lPFRadiusMin=1000
  nonFocusAzimuth=30
  nonFocusVolumeAttenuation=0
  stereoSpread=2048
  attenuationShapeExtents {x=1024}
  customAttenuationCurve {editorCurveData {keys [<at 0.1 0>,<at 0.25 1>,<at 1 0>]}}
 }
}
