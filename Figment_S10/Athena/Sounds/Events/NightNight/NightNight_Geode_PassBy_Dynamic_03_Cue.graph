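# Sound cue for a dynamic geode pass-by whoosh in the NightNight event, variant 03.
# A Random node picks between three SpaceRocks whoosh waves with equal weight.
# An inline attenuation override turns off distance volume falloff, despite carrying a custom curve, and sets a wide stereo spread and focus interpolation.
# It routes to the extra-high-priority events sound class.
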SoundCue /Figment_S10/Athena/Sounds/Events/NightNight/NightNight_Geode_PassBy_Dynamic_03_Cue
root=Random_0
nodes [
 {id=Random_0, children [WavePlayer_11, WavePlayer_12, WavePlayer_0], properties {weights [1,1,1]}}
 {id=WavePlayer_0, properties {soundWaveAssetPtr=/Figment_S10/Sounds/Fort_GamePlay_Sounds/Wargames/SpaceRocks_Whoosh_03}}
 {id=WavePlayer_11, properties {soundWaveAssetPtr=/Figment_S10/Sounds/Fort_GamePlay_Sounds/Wargames/SpaceRocks_Whoosh_07}}
 {id=WavePlayer_12, properties {soundWaveAssetPtr=/Figment_S10/Sounds/Fort_GamePlay_Sounds/Wargames/SpaceRocks_Whoosh_08}}
]
properties {
 bOverrideAttenuation=true
 maxDistance=2097152
 soundClassObject=/Figment_S10/Athena/Sounds/Events/SoundClasses/Fort_Events_Priority_ExtHigh
 volumeMultiplier=1
 attenuationOverrides {
  bAttenuate=false
  bEnableFocusInterpolation=true
  falloffDistance=100000
  focusAttackInterpSpeed=0.5
  focusAzimuth=5
  focusReleaseInterpSpeed=0.25
  lPFFrequencyAtMax=500
  lPFRadiusMax=2000
  lPFRadiusMin=1000
  nonFocusAzimuth=30
  nonFocusVolumeAttenuation=0
  stereoSpread=2048
  attenuationShapeExtents {x=1024}
  customAttenuationCurve {editorCurveData {keys [<at 0.1 0>,<at 0.25 1>,<at 1 0>]}}
 }
}
