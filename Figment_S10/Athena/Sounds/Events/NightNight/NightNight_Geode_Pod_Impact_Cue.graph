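# Sound cue for the Geode pod impact in the NightNight event, built from layered meteor landing sounds.
# A top mixer blends a close impact wave with a slightly quieter distant landing wave, each behind its own attenuation asset.
# A third close-attenuation branch with an overridden, very large falloff distance is empty, as is one input of the close submixer.
# The cue plays louder than default in a high-priority events sound class.
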SoundCue /Figment_S10/Athena/Sounds/Events/NightNight/NightNight_Geode_Pod_Impact_Cue
root=Mixer_1
nodes [
 {
  id=Attenuation_2
  children [WavePlayer_0]
  properties {attenuationSettings=/Figment_S10/Athena/Sounds/Events/NightNight/Attenuations/NightNight_Geode_Distant_Att}
 }
 {
  id=Attenuation_3
  children [Mixer_2]
  properties {attenuationSettings=/Figment_S10/Athena/Sounds/Events/NightNight/Attenuations/NightNight_Close_Pod_Att}
 }
 {
  id=Attenuation_5
  children [null]
  properties {
   attenuationSettings=/Figment_S10/Athena/Sounds/Events/NightNight/Attenuations/NightNight_Close_Att
   bOverrideAttenuation=true
   attenuationOverrides {falloffDistance=2000000}
  }
 }
 {id=Mixer_1, children [Attenuation_3, Attenuation_2, Attenuation_5], properties {inputVolume [1,0.8,1]}}
 {id=Mixer_2, children [null,WavePlayer_3], properties {inputVolume [1,1]}}
 {id=WavePlayer_0, properties {soundWaveAssetPtr=/Figment_S10/Sounds/Fort_GamePlay_Sounds/Meteor/Meteor_Land_Distant}}
 {id=WavePlayer_3, properties {soundWaveAssetPtr=/Figment_S10/Sounds/Fort_GamePlay_Sounds/Meteor/Meteor_Impact_01}}
]
properties {
 maxDistance=2097152
 soundClassObject=/Figment_S10/Athena/Sounds/Events/SoundClasses/Fort_Events_Priority_ExtHigh
 volumeMultiplier=1.25
 attenuationOverrides {
  bEnableFocusInterpolation=true
  bEnableListenerFocus=true
  falloffDistance=100000
  focusAttackInterpSpeed=0.5
  focusAzimuth=5
  focusReleaseInterpSpeed=0.25
  lPFFrequencyAtMax=500
  lPFRadiusMax=2000
  lPFRadiusMin=1000
  nonFocusAzimuth=30
  nonFocusVolumeAttenuation=0
  stereoSpread=2048
  attenuationShapeExtents {x=1024}
  customAttenuationCurve {editorCurveData {keys [<at 0.1 0>,<at 0.25 1>,<at 1 0>]}}
 }
}
