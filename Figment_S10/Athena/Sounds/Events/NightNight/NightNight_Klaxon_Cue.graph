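# Klaxon sound cue for the NightNight event, built from Geode event klaxon waves and attenuations on the Athena SFX class at high priority.
# Each klaxon plays three times, immediately and after fixed three- and six-second delays, through parallel mixer inputs.
# Close and distant layers get separate attenuations and are mixed together, with the close layer slightly louder.
# A quality-level node sends the first two levels to the standard waves and the last two to "_opt" variants.
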SoundCue /Figment_S10/Athena/Sounds/Events/NightNight/NightNight_Klaxon_Cue
root=QualityLevel_0
nodes [
 {id=Attenuation_0, children [Mixer_1], properties {attenuationSettings=/Figment_S10/Athena/Sounds/Events/Geode/Geode_Klaxxon_Close_Att}}
 {
  id=Attenuation_1
  children [Mixer_2]
  properties {attenuationSettings=/Figment_S10/Athena/Sounds/Events/Geode/Geode_Distant_Frequent_Klaxxon_Att}
 }
 {id=Attenuation_2, children [Mixer_4], properties {attenuationSettings=&Geode_Klaxxon_Close_Att}}
 {
  id=Attenuation_3
  children [Mixer_5]
  properties {attenuationSettings=&Geode_Distant_Frequent_Klaxxon_Att}
 }
 {id=Delay_0, children [WavePlayer_2], properties {delayMax=3, delayMin=3}}
 {id=Delay_1, children [WavePlayer_2], properties {delayMax=6, delayMin=6}}
 {id=Delay_2, children [WavePlayer_1], properties {delayMax=3, delayMin=3}}
 {id=Delay_3, children [WavePlayer_1], properties {delayMax=6, delayMin=6}}
 {id=Delay_4, children [WavePlayer_5], properties {delayMax=3, delayMin=3}}
 {id=Delay_5, children [WavePlayer_5], properties {delayMax=6, delayMin=6}}
 {id=Delay_6, children [WavePlayer_6], properties {delayMax=3, delayMin=3}}
 {id=Delay_7, children [WavePlayer_6], properties {delayMax=6, delayMin=6}}
 {id=Mixer_0, children [Attenuation_0, Attenuation_1], properties {inputVolume [1.2,1]}}
 {id=Mixer_1, children [WavePlayer_2,Delay_0,Delay_1], properties {inputVolume [1,1,1]}}
 {id=Mixer_2, children [WavePlayer_1,Delay_2,Delay_3], properties {inputVolume [1,1,1]}}
 {id=Mixer_3, children [Attenuation_2, Attenuation_3], properties {inputVolume [1.2,1]}}
 {id=Mixer_4, children [WavePlayer_5,Delay_4,Delay_5], properties {inputVolume [1,1,1]}}
 {id=Mixer_5, children [WavePlayer_6,Delay_6,Delay_7], properties {inputVolume [1,1,1]}}
 {id=QualityLevel_0, children [Mixer_0,Mixer_0,Mixer_3,Mixer_3]}
 {id=WavePlayer_1, properties {soundWaveAssetPtr=/Figment_S10/Athena/Sounds/Events/Geode/Geode_Tear_Klaxxon_Distant_19}}
 {id=WavePlayer_2, properties {soundWaveAssetPtr=/Figment_S10/Athena/Sounds/Events/Geode/Geode_Tear_Klaxxon_Close_18}}
 {id=WavePlayer_5, properties {soundWaveAssetPtr=/Figment_S10/Athena/Sounds/Events/Geode/Geode_Tear_Klaxxon_Close_18_opt}}
 {id=WavePlayer_6, properties {soundWaveAssetPtr=/Figment_S10/Athena/Sounds/Events/Geode/Geode_Tear_Klaxxon_Distant_19_opt}}
]
properties {maxDistance=32000, priority=100, soundClassObject=/Figment_S10/Sounds/FortSoundClassesAndModes/Fort_Athena_SFX, volumeMultiplier=1}
