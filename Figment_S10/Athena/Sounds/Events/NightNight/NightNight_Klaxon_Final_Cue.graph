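# Sound cue for the NightNight event's final klaxon, reusing Geode event klaxon waves and attenuations on the Athena SFX class.
# A shared mixer layers close and distant klaxons, each with its own attenuation and a quality switch between standard and "_opt" waves.
# The root mixer plays that layer four times at three-second intervals by reusing it through delay nodes.
# An empty, non-attenuating branch carries a huge falloff distance, matching the cue's very large max distance.
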SoundCue /Figment_S10/Athena/Sounds/Events/NightNight/NightNight_Klaxon_Final_Cue
root=Mixer_9
nodes [
 {id=Attenuation_0, children [null], properties {bOverrideAttenuation=true, attenuationOverrides {bAttenuate=false, falloffDistance=2000000}}}
 {
  id=Attenuation_8
  children [QualityLevel_2]
  properties {attenuationSettings=/Figment_S10/Athena/Sounds/Events/Geode/Geode_Klaxxon_Close_Att}
 }
 {
  id=Attenuation_9
  children [QualityLevel_3]
  properties {attenuationSettings=/Figment_S10/Athena/Sounds/Events/Geode/Geode_Distant_Klaxxon_Att}
 }
 {id=Delay_10, children [Mixer_8], properties {delayMax=9, delayMin=9}}
 {id=Delay_8, children [Mixer_8], properties {delayMax=3, delayMin=3}}
 {id=Delay_9, children [Mixer_8], properties {delayMax=6, delayMin=6}}
 {id=Mixer_8, children [Attenuation_8, Attenuation_9, Attenuation_0], properties {inputVolume [1,1,1]}}
 {id=Mixer_9, children [Mixer_8,Delay_8,Delay_9,Delay_10], properties {inputVolume [1,1,1,1]}}
 {id=QualityLevel_2, children [WavePlayer_9, WavePlayer_9, WavePlayer_10, WavePlayer_10]}
 {id=QualityLevel_3, children [WavePlayer_8, WavePlayer_8, WavePlayer_11, WavePlayer_11]}
 {id=WavePlayer_10, properties {soundWaveAssetPtr=/Figment_S10/Athena/Sounds/Events/Geode/Geode_Tear_Klaxxon_Close_18_opt}}
 {id=WavePlayer_11, properties {soundWaveAssetPtr=/Figment_S10/Athena/Sounds/Events/Geode/Geode_Tear_Klaxxon_Distant_19_opt}}
 {id=WavePlayer_8, properties {soundWaveAssetPtr=/Figment_S10/Athena/Sounds/Events/Geode/Geode_Tear_Klaxxon_Distant_19}}
 {id=WavePlayer_9, properties {soundWaveAssetPtr=/Figment_S10/Athena/Sounds/Events/Geode/Geode_Tear_Klaxxon_Close_18}}
]
properties {maxDistance=2097152, priority=100, soundClassObject=/Figment_S10/Sounds/FortSoundClassesAndModes/Fort_Athena_SFX, volumeMultiplier=1}
