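# Sound cue for the looping holotable launch warning used in the Figment Season 10 NightNight event.
# A quality-level switch plays the standard looping wave on its first two levels and an "_opt" variant on the last two.
# The switch feeds a root mixer whose second input is empty.
# Output plays at reduced volume in the Fort_Gadgets sound class, with the NightNight_Pre_Att attenuation asset assigned.
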SoundCue /Figment_S10/Athena/Sounds/Events/NightNight/NightNight_LaunchWarning_Cue
root=Mixer_0
nodes [
 {id=Mixer_0, children [QualityLevel_0, null], properties {inputVolume [1,1]}}
 {id=QualityLevel_0, children [WavePlayer_0,WavePlayer_0,WavePlayer_1,WavePlayer_1]}
 {id=WavePlayer_0, properties {bLooping=true, soundWaveAssetPtr=/Figment_S10/Sounds/Fort_Gadgets/Holotable/Holotable_LaunchWarning_04}}
 {id=WavePlayer_1, properties {bLooping=true, soundWaveAssetPtr=/Figment_S10/Sounds/Fort_Gadgets/Holotable/Holotable_LaunchWarning_04_opt}}
]
properties {
 attenuationSettings=/Figment_S10/Athena/Sounds/Events/NightNight/Attenuations/NightNight_Pre_Att
 maxDistance=2097152
 soundClassObject=&Fort_Gadgets
 volumeMultiplier=0.2
 attenuationOverrides {falloffDistance=32000, nonSpatializedRadiusStart=3000, stereoSpread=256, attenuationShapeExtents {x=0}}
}
